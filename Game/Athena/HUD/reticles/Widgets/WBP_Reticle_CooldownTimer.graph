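# Reticle widget that shows a cooldown ring for the item of the weapon the player holds at construction.
# A looping query timer reads that item's remaining cooldown, then SetCooldownActive plays the intro animation, broadcasts OnCooldownStateChanged and starts a faster progress timer.
# That timer drives the ring material's CooldownPercent from game time and ends the cooldown at completion.
# The experience's NoCooldowns setting forces the ring hidden.
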WidgetBlueprint /Game/Athena/HUD/reticles/Widgets/WBP_Reticle_CooldownTimer
animations [Cooldown_Intro {
  binding_count=1
  length_seconds=0.4
  object=&self:Cooldown_Intro
  sequence {bindings [CooldownRing {
     id=#1
     possessed_class=Image
     tracks [
      {
       type=MovieSceneFloatTrack
       display_name=Render Opacity
       sections [{type=MovieSceneFloatSection, channels [None {type=MovieSceneScriptingFloatChannel, keys [[0,0],[5,1]]}]}]
      }
      {
       type=MovieScene2DTransformTrack
       sections [{
         type=MovieScene2DTransformSection
         channels [
          Scale.X {type=MovieSceneScriptingFloatChannel, keys [[0,0.5],[8,1]]}
          Scale.Y {type=MovieSceneScriptingFloatChannel, keys [[0,0.5],[8,1]]}
         ]
        }]
      }
     ]
    }]}
 }]
graphs [
 EventGraph {nodes [
   Event Destruct()
   > Parent: Destruct()
   > SetCooldownActive!(execute, self, IsActive:bool=false, ShouldContinueQuery:bool=false) -> (then)
   > Clear and Invalidate Timer by Handle(Handle=Get QueryCooldownStateTimerHandle() -> (QueryCooldownStateTimerHandle:Timer Handle Structure))
   Event Construct()
   > Parent: Construct()
   > SetVisibility(self=Get CooldownRing() -> (CooldownRing:Image Object Reference), InVisibility=Collapsed)
   > n1 = Cast To FortPlayerController(Object=GetOwningPlayer())
   > Set CachedFortController(CachedFortController:Fort Player Controller Object Reference=@n1) -> (Output_Get:Fort Player Controller Object Reference)
   n2 = GetPlayerPawn!(self:Fort Player Controller Object Reference=@n1) -> (ReturnValue:Fort Player Pawn Object Reference)
   > Is Valid(InputObject=@n2)
   n3 = GetCurrentWeapon!(self:Fort Pawn Object Reference=@n2) -> (ReturnValue:Fort Weapon Object Reference)
   > Is Valid(InputObject=@n3)
   > Set CachedFortItem(CachedFortItem:Fort Item Object Reference=GetInventoryItem!(self:Fort Weapon Object Reference=@n3) -> (ReturnValue:Fort Item Object Reference)) -> (Output_Get:Fort Item Object Reference)
   > StartCooldownQueryTimerEvent()
  ]}
 GetIsCooldownActive {nodes [
   GetIsCooldownActive() {is=title K2Node_FunctionEntry}
   > Return Node!(execute, IsCooldownActive:bool=Get IsCooldownActive() -> (IsCooldownActive:bool))
  ]}
 OnCooldownStateChanged {nodes [OnCooldownStateChanged!() -> (then, IsCooldownActive:bool=~) {is=title K2Node_FunctionEntry}]}
 QueryCooldownState {nodes [
   QueryCooldownState() {is=title K2Node_FunctionEntry}
   n1 = Get CachedFortItem() -> (CachedFortItem:Fort Item Object Reference)
   n2 = Get CachedFortController() -> (CachedFortController:Fort Player Controller Object Reference)
   > n3 = Branch(Condition=AND Boolean(A=IsValid(Object=@n1), B=IsValid(Object=@n2)))
   n4 = GetCoolDownTimeRemaining!(self:Fort Item Object Reference=Get CachedFortItem(), FortPC:Fort Player Controller Object Reference=Get CachedFortController()) -> (CooldownDuration:float, ReturnValue:float)
   > Set TempCooldownDuration(TempCooldownDuration:float=@n4) -> (Output_Get:float)
   > n5 = Set TempInitialTimeRemaining(TempInitialTimeRemaining:float=@n4.ReturnValue) -> (Output_Get:float)
   > n6 = Branch(Condition="float > float"(A=@n5))
   > n7 = Branch(Condition=Get IsCooldownActive())
   Set CooldownDuration(execute=@n7.else, CooldownDuration:float=Get TempCooldownDuration() -> (TempCooldownDuration:float)) -> (Output_Get:float)
   > n8 = Set CooldownInitialTimeRemaining(CooldownInitialTimeRemaining:float=Get TempInitialTimeRemaining() -> (TempInitialTimeRemaining:float)) -> (Output_Get:float)
   > SetCooldownActive(IsActive="float > float"(A=@n8), ShouldContinueQuery=true)
   SetCooldownActive(execute=@n6.else, IsActive=false, ShouldContinueQuery=true)
   SetCooldownActive(execute=@n3.else, IsActive=false, ShouldContinueQuery=false)
  ]}
 SetCooldownActive {nodes [
   SetCooldownActive!() -> (then, IsActive:bool=~, ShouldContinueQuery:bool=true) {is=title K2Node_FunctionEntry}
   n1 = GetOwningPlayer()
   n2 = Get NoCooldowns!(self:Fort Experience Settings Component Object Reference=GetExperienceSettingsComponentFromGameState!(None:%=@n1) -> (ReturnValue:Fort Experience Settings Component Object Reference)) -> (NoCooldowns:Scalable Float Structure)
   > n3 = Branch(Condition=NOT Boolean(A=GetValueAsBoolWithContextParam!(WorldContextObject:Object Reference=@n1, Input:"Scalable Float Structure (by ref)"=@n2, Level:float=0) -> (ReturnValue:bool)))
   n4 = Get IsActive() -> (IsActive:bool)
   > Branch(Condition="Not Equal (Boolean)"(A=@n4, B=Get IsCooldownActive()) -> (ReturnValue=~))
   > n5 = Set IsCooldownActive(IsCooldownActive:bool=@n4) -> (Output_Get:bool)
   > Call On Cooldown State Changed!(execute:Exec, self:WBP Reticle Cooldown Timer Object Reference, IsCooldownActive:bool=@n5) -> (then:Exec) {is=title K2Node_CallDelegate}
   > n6 = Branch(Condition=@n5)
   n7 = Get CooldownRing()
   > SetVisibility(self=@n7, InVisibility=SelfHitTestInvisible)
   > PlayAnimation(InAnimation=Get Cooldown_Intro() -> (Cooldown_Intro:Widget Animation Object Reference))
   n8 = Get CooldownInitialTimeRemaining() -> (CooldownInitialTimeRemaining:float)
   n9 = Get CooldownDuration() -> (CooldownDuration:float)
   > Set StartTime(StartTime:double=float - float(A=GetGameTimeInSeconds(), B=float - float(A=@n9, B=@n8))) -> (Output_Get:double)
   > Branch(Condition=Get ShouldContinueQuery() -> (ShouldContinueQuery:bool))
   > StartActiveCooldownTimerEvent()
   SetVisibility(execute=[@n12.then, @n6.else], self=@n7, InVisibility=Collapsed)
   > Clear and Invalidate Timer by Handle(Handle=Get UpdateCooldownProgressTimerHandle() -> (UpdateCooldownProgressTimerHandle:Timer Handle Structure))
   > n10 = Branch(Condition=Get ShouldContinueQuery())
   Clear and Invalidate Timer by Handle(execute=@n10.else, Handle=Get QueryCooldownStateTimerHandle())
   n11 = Set IsCooldownActive(execute=@n3.else, IsCooldownActive=false)
   > n12 = Call On Cooldown State Changed(IsCooldownActive=@n11)
  ]}
 SetCooldownProgress {nodes [
   n1 = SetCooldownProgress!() -> (then, Value:float=~) {is=title K2Node_FunctionEntry}
   > n2 = GetDynamicMaterial(self=Get CooldownRing())
   > SetScalarParameterValue(self=@n2, ParameterName=CooldownPercent, Value=@n1)
  ]}
 StartActiveCooldownTimerEvent {nodes [
   StartActiveCooldownTimerEvent() {is=title K2Node_FunctionEntry}
   > Clear and Invalidate Timer by Handle(Handle=Get UpdateCooldownProgressTimerHandle())
   > n1 = Set Timer by Event(Delegate=Create Event(), Time=Get UpdateCooldownFrequency() -> (UpdateCooldownFrequency:float), bLooping=true)
   > Set UpdateCooldownProgressTimerHandle(UpdateCooldownProgressTimerHandle:Timer Handle Structure=@n1) -> (Output_Get:Timer Handle Structure)
  ]}
 StartCooldownQueryTimerEvent {nodes [
   StartCooldownQueryTimerEvent() {is=title K2Node_FunctionEntry}
   > Clear and Invalidate Timer by Handle(Handle=Get QueryCooldownStateTimerHandle())
   > n1 = Set Timer by Event(Delegate=Create Event(), Time=Get QueryCooldownFrequency() -> (QueryCooldownFrequency:float), bLooping=true)
   > Set QueryCooldownStateTimerHandle(QueryCooldownStateTimerHandle:Timer Handle Structure=@n1) -> (Output_Get:Timer Handle Structure)
  ]}
 UpdateCooldownProgress {nodes [
   UpdateCooldownProgress() {is=title K2Node_FunctionEntry}
   n1 = GetGameTimeInSeconds()
   n2 = Get StartTime() -> (StartTime:double)
   > n3 = Set PercentComplete(PercentComplete:double=float / float(A=float - float(A=@n1, B=@n2), B=Get CooldownDuration())) -> (Output_Get:double)
   > SetCooldownProgress!(execute, self, Value:float=@n3) -> (then) {is=title K2Node_CallFunction}
   > Branch(Condition="float >= float"(A=Get PercentComplete() -> (PercentComplete:double), B=1))
   > SetCooldownActive(IsActive=false, ShouldContinueQuery=true) {is=title K2Node_CallFunction}
  ]}
]
variables [
 CachedFortController {type=FortPlayerController, default=null}
 CachedFortItem {type=FortItem, default=null}
 CooldownDuration {type=number}
 CooldownInitialTimeRemaining {type=number}
 IsCooldownActive {type=boolean}
 OnCooldownStateChanged {}
 QueryCooldownFrequency {type=number, default=0.5}
 QueryCooldownStateTimerHandle {type=TimerHandle}
 StartTime {type=number}
 UpdateCooldownFrequency {type=number, default=0.1}
 UpdateCooldownProgressTimerHandle {type=TimerHandle}
]
widgets {
 info {inheritedWidgetCount=0, parentClass=&UserWidget, rootWidgetClass=&Image, widgetCount=1}
 widgets [CooldownRing {
   widgetClassPath=&Image
   bIsVariable=true
   widget=CooldownRing
   properties {
    visibility=HitTestInvisible
    brush {
     bIsDynamicallyLoaded=false
     drawAs=Image
     imageType=NoImage
     mirroring=NoMirror
     resourceName=None
     resourceObject=/Game/Athena/HUD/reticles/Materials/MI_UI_CooldownRing_Default
     tiling=NoTile
     imageSize=<56 56>
     margin=<ltrb 0 0 0 0>
     outlineSettings {
      bUseBrushTransparency=false
      roundingType=HalfHeightRadius
      width=0
      color {colorUseRule=UseColor_Specified, specifiedColor=<rgba 0 0 0 0>}
      cornerRadii=<0 0 0 0>
     }
     tintColor {colorUseRule=UseColor_Specified, specifiedColor=<rgba 1 1 1 1>}
     uVRegion {bIsValid=false, max=<0 0>, min=<0 0>}
    }
   }
  }]
}
